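# Physics asset for the Ohana Male pickaxe, giving capsule bodies to its two-handed handle and a hand-shaped chain of dyn_ bones.
# The chain runs from the handle through two hand bodies, with the thumb branching from the first and four fingers from the second, each digit having two segments.
# Constraints allow small swing angles and zero twist, keeping any simulated motion of the hand and fingers tight.
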PhysicsAsset /BRCosmetics/Weapons/FORT_Melee/Pickaxe_Ohana_Male/Meshes/Ohana_Male_Axe_Physics
bodies {
 Melee_TwoHanded_Handle [{shapeType=Capsule, length=16.9297, radius=2, shapeName=Melee_TwoHanded_Handle_capsule}]
 dyn_Hand_1 [{shapeType=Capsule, length=0, radius=6.5697, shapeName=dyn_Hand_1_capsule, center=<2.5 0 0>, rotation=<rot 90 0 0>}]
 dyn_Hand_2 [{shapeType=Capsule, length=7.066, radius=5.1663, shapeName=dyn_Hand_2_capsule, center=<2.5 0 0>}]
 dyn_Index_1 [{shapeType=Capsule, length=2.9612, radius=2, shapeName=dyn_Index_1_capsule, center=<3.2 0 0>, rotation=<rot 90 0 0>}]
 dyn_Index_2 [{shapeType=Capsule, length=0, radius=2.6248, shapeName=dyn_Index_2_capsule, center=<2.5 0 0>, rotation=<rot 90 0 0>}]
 dyn_Middle_1 [{shapeType=Capsule, length=4.742, radius=2, shapeName=dyn_Middle_1_capsule, center=<4 0 0>, rotation=<rot 90 0 0>}]
 dyn_Middle_2 [{shapeType=Capsule, length=2.3415, radius=2.54, shapeName=dyn_Middle_2_capsule, center=<4 0 0>, rotation=<rot 90 0 0>}]
 dyn_Pinky_1 [{shapeType=Capsule, length=2, radius=2, shapeName=dyn_Pinky_1_capsule, center=<2.5 0 0>, rotation=<rot 90 0 0>}]
 dyn_Pinky_2 [{shapeType=Capsule, length=0, radius=2, shapeName=dyn_Pinky_2_capsule, center=<2.2 0 0>, rotation=<rot 90 0 0>}]
 dyn_Ring_1 [{shapeType=Capsule, length=3.8894, radius=2, shapeName=dyn_Ring_1_capsule, center=<4 0 0>, rotation=<rot 90 0 0>}]
 dyn_Ring_2 [{shapeType=Capsule, length=1.4443, radius=2.4139, shapeName=dyn_Ring_2_capsule, center=<3 0 0>, rotation=<rot 90 0 0>}]
 dyn_Thumb_1 [{shapeType=Capsule, length=2.948, radius=2.4473, shapeName=dyn_Thumb_1_capsule, center=<2.5 0 0>, rotation=<rot 90 0 0>}]
 dyn_Thumb_2 [{shapeType=Capsule, length=0, radius=2, shapeName=dyn_Thumb_2_capsule, center=<2 0 0>, rotation=<rot 90 0 0>}]
}
constraints [
 {bone1Name=dyn_Hand_1, bone2Name=Melee_TwoHanded_Handle, swing1LimitDegrees=2.5, swing2LimitDegrees=1, twistLimitDegrees=0}
 {bone1Name=dyn_Hand_2, bone2Name=dyn_Hand_1, swing1LimitDegrees=10, swing2LimitDegrees=10, twistLimitDegrees=0}
 {bone1Name=dyn_Index_1, bone2Name=dyn_Hand_2, swing1LimitDegrees=10, swing2LimitDegrees=5, twistLimitDegrees=0}
 {bone1Name=dyn_Index_2, bone2Name=dyn_Index_1, swing1LimitDegrees=6, swing2LimitDegrees=7, twistLimitDegrees=0}
 {bone1Name=dyn_Middle_1, bone2Name=dyn_Hand_2, swing1LimitDegrees=11, swing2LimitDegrees=10, twistLimitDegrees=0}
 {bone1Name=dyn_Middle_2, bone2Name=dyn_Middle_1, swing1LimitDegrees=8, swing2LimitDegrees=7, twistLimitDegrees=0}
 {bone1Name=dyn_Ring_1, bone2Name=dyn_Hand_2, swing1LimitDegrees=11, swing2LimitDegrees=10, twistLimitDegrees=0}
 {bone1Name=dyn_Ring_2, bone2Name=dyn_Ring_1, swing1LimitDegrees=8.5, swing2LimitDegrees=7, twistLimitDegrees=0}
 {bone1Name=dyn_Pinky_1, bone2Name=dyn_Hand_2, swing1LimitDegrees=11.25, swing2LimitDegrees=10, twistLimitDegrees=0}
 {bone1Name=dyn_Pinky_2, bone2Name=dyn_Pinky_1, swing1LimitDegrees=6.5, swing2LimitDegrees=7, twistLimitDegrees=0}
 {bone1Name=dyn_Thumb_1, bone2Name=dyn_Hand_1, swing1LimitDegrees=10, swing2LimitDegrees=10, twistLimitDegrees=0}
 {bone1Name=dyn_Thumb_2, bone2Name=dyn_Thumb_1, swing1LimitDegrees=10, swing2LimitDegrees=10, twistLimitDegrees=0}
]
